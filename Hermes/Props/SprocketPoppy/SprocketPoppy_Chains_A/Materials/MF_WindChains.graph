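# Material function that animates hanging chains in wind through world position offset, with a switch for U-shaped chains.
# A pivot-based rotation swings vertices on a local or wandering axis, plus a sine wobble masked by height gradients.
# World position offsets time per instance; Nanite takes the pivot from custom primitive data.
# It outputs material attributes with offset, rotated normal and debug emissive; storm blends are present but unconnected.
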MaterialFunction /Hermes/Props/SprocketPoppy/SprocketPoppy_Chains_A/Materials/MF_WindChains
graph {nodes [
  n1 = Add(
   A=ScalarParameter() {group=WindChains, parameterName=-----------, sortPriority=15}
   B=ScalarParameter() {group=WindChains, parameterName=------------, sortPriority=4}
  )
  n2 = Add(
   A=ScalarParameter() {group=WindChains, parameterName=-------------, sortPriority=3}
   B=ScalarParameter() {group=WindChains, parameterName=--------------, sortPriority=2}
  )
  Add(B=Multiply(A=Add(A=@n1, B=@n2)) {constB=0}) {constB=0}
  n3 = NamedRerouteUsage()
  n4 = Multiply(A=ScalarParameter() {defaultValue=5, group=WindChains, parameterName=Wobloe Intensity, sliderMax=5, sortPriority=4}) {constB=5}
  n5 = Multiply(A=NamedRerouteUsage()) {constB=-1}
  n6 = Constant() {r=0.63}
  n7 = Multiply(A=Add(A=NamedRerouteUsage(), B=Multiply(A=@n5, B=@n6)), B=Constant() {r=1}) {constB=5}
  n8 = Sine(None=@n7)
  n9 = Add(A=NamedRerouteUsage(), B=Multiply(A=NamedRerouteUsage()) {constB=-1})
  n10 = Sine(None=Multiply(A=@n9, B=ScalarParameter() {defaultValue=0.5, group=WindChains, parameterName=Wooble Wave Frequency, sliderMax=1, sortPriority=4}))
  n11 = Multiply(
   A=LinearInterpolate(A=Constant3Vector() {constant=<rgba 0 1 0 0>}, B=Constant3Vector() {constant=<rgba 1 0 0 0>}, Alpha=ConstantBiasScale(None=@n8))
   B=@n10
  )
  n12 = NamedRerouteUsage()
  n13 = Multiply(
   A=Multiply(A=Multiply(A=@n4, B=@n11), B=Clamp(None=@n12) {clampMode=CMODE_ClampMin, minDefault=0.1})
   B=NamedRerouteUsage() {desc=Fades WPO Wooble}
  )
  n14 = Constant() {r=1}
  n15 = NamedRerouteUsage()
  n16 = NamedRerouteUsage()
  n17 = Constant()
  n18 = Constant() {r=800}
  n19 = StaticSwitchParameter(
   True=Constant() {r=1}
   False=LinearInterpolate(Alpha=MaterialFunctionCall(Input=@n16, Input Low=@n17, Input High=@n18) {materialFunction=&RemapValue_0-1}.Result) {constA=1, constB=0.1}
  ) {defaultValue=true, group=WindChains, parameterName=U Shaped Chain, sortPriority=0}
  n20 = NamedRerouteUsage()
  n21 = ScalarParameter() {defaultValue=0.2, group=WindChains, parameterName=Swing Max Angle, sliderMax=0.5, sortPriority=2}
  n22 = Constant() {r=1}
  n23 = Multiply(A=@n19, B=Multiply(A=Multiply(A=@n20, B=@n21) {constB=0.2}, B=@n22) {constA=0.01, desc=Speed x Intensity x Angle})
  n24 = NamedRerouteUsage()
  n25 = Multiply(
   A=NamedRerouteUsage()
   B=Multiply(
    A=Add(A=@n24) {constB=0.1}
    B=ScalarParameter() {defaultValue=0.5, group=WindChains, parameterName=Swing Max Speed, sliderMax=1, sortPriority=2}
   ) {constB=0.5}
  )
  n26 = Sine(None=@n25) {period=0.5}
  n27 = Transform(None=Constant3Vector() {constant=<rgba 0 1 0 0>}) {desc=Localized Axis, transformSourceType=TRANSFORMSOURCE_Local}
  n28 = NamedRerouteUsage()
  n29 = Multiply(
   A=NamedRerouteUsage()
   B=Multiply(
    A=Add(A=@n28) {constB=0.1}
    B=ScalarParameter() {defaultValue=0.1, group=WindChains, parameterName=Swing Max RandomDirection Speed, sliderMax=0.5, sortPriority=2}
   ) {constB=0.1}
  )
  n30 = Sine(None=@n29)
  n31 = MaterialFunctionCall(
   1=Constant3Vector() {constant=<rgba 1 0 0 0>}
   2=Constant3Vector() {constant=<rgba 0 1 0 0>}
   3=Constant3Vector() {constant=<rgba -1 0 0 0>}
   A=ConstantBiasScale(None=@n30) {desc=Direction Alpha}
  ) {desc=Rotation Axis Lerp, materialFunction=&Lerp_Multiple_Float3}
  n32 = MaterialFunctionCall!(
   RotationAngle=Multiply(A=NamedRerouteUsage(), B=Multiply(A=@n23, B=@n26))
   NormalizedRotationAxis=Normalize(VectorInput=StaticSwitchParameter(True=@n27, False=@n31.Lerp 3 Inputs) {defaultValue=true, group=WindChains, parameterName=U Shaped Chain, sortPriority=0})
   PivotPoint=NamedRerouteUsage()
   Position=WorldPosition() {worldPositionShaderOffset=WPT_CameraRelative}.XYZ
   Normal=NamedRerouteUsage()
   TangentU
   TangentV
  ) {materialFunction=/Game/Characters/Player/Male/Medium/Bodies/M_MED_Inferno/FX/Mat/RotateVertexAboutAxis}
  n33 = Add(
   A=@n13
   B=Multiply(
    A=StaticSwitchParameter(True=@n14, False=@n15) {defaultValue=true, group=WindChains, parameterName=U Shaped Chain, sortPriority=0}
    B=@n32.WPO
   )
  )
  n34 = NamedRerouteUsage() {desc=Rred}
  n35 = StaticSwitchParameter(True=Constant(), False=NamedRerouteUsage()) {defaultValue=true, group=WindChains, parameterName=U Shaped Chain, sortPriority=0}
  n36 = StaticSwitchParameter(
   True=AppendVector(A=AppendVector(A=@n34, B=@n35) {desc=Yellow}, B=Constant())
   False=StaticSwitchParameter(True=NamedRerouteUsage(), False=Constant()) {group=WindChains, parameterName=Debug Wooble, sortPriority=0}
  ) {group=WindChains, parameterName=Debug Gradient, sortPriority=0}
  FunctionOutput(None=SetMaterialAttributes!(
    MaterialAttributes=NamedRerouteUsage()
    WorldPositionOffset=Add(A=@n3, B=@n33)
    Normal=Transform(None=@n32.Normal) {transformSourceType=TRANSFORMSOURCE_World, transformType=TRANSFORM_Tangent}
    EmissiveColor=VertexInterpolator(VS=@n36).PS
   ) {attributeSetTypes [#1, #2, #3]})
  LinearInterpolate(A=@n22, B=Constant() {r=3}, Alpha=NamedRerouteUsage())
  LinearInterpolate(A=@n8, B=Sine(None=@n7), Alpha=NamedRerouteUsage()) {constA=1}
  n37 = Multiply(A=ScalarParameter() {defaultValue=1, group=WindChains, parameterName=STORM Swing Intensity, sliderMax=1, sortPriority=16}) {constB=0.5}
  LinearInterpolate(A=@n30, B=Sine(None=Multiply(A=@n29, B=@n37)), Alpha=NamedRerouteUsage()) {constA=1}
  n38 = ScalarParameter() {defaultValue=1.2, group=WindChains, parameterName=STORM Wooble Wave Frequency, sliderMax=1.5, sortPriority=16}
  LinearInterpolate(A=@n10, B=Sine(None=Multiply(A=@n9, B=@n38) {constB=5}), Alpha=NamedRerouteUsage()) {constA=1}
  n39 = ScalarParameter() {defaultValue=1, group=WindChains, parameterName=STORM Swing Intensity, sliderMax=3, sortPriority=16}
  LinearInterpolate(A=@n26, B=Sine(None=Multiply(A=@n25, B=@n39)) {period=0.5}, Alpha=NamedRerouteUsage()) {constA=1}
  n40 = GetMaterialAttributes(None=FunctionInput() {inputType=FunctionInput_MaterialAttributes}) {attributeGetTypes [#1, #2]}
  NamedRerouteDeclaration(None=Transform(None=@n40.Normal)) {name=WS Normal, nodeColor=<rgba 0 0.0523 1 1>}
  n41 = Ceil(None=TransformPosition(None=Constant3Vector()) {transformType=TRANSFORMPOSSOURCE_World})
  n42 = Constant() {desc="LeapOfFaith ", r=10000}
  n43 = MaterialFunctionCall(Float3=Multiply(A=Divide(A=@n41, B=@n42)) {constB=11.17}) {materialFunction=&BreakOutFloat3Components}
  n44 = Frac(None=Add(A=@n43.R, B=@n43.R))
  NamedRerouteDeclaration(None=Add(A=Multiply(A=@n44) {constB=120}, B=Time())) {name=Random time, nodeColor=<rgba 0 1 0.7422 1>}
  NamedRerouteDeclaration(None=@n19) {name=Fixture Mask, nodeColor=<rgba 1 0 0.3281 1>}
  n45 = VectorParameter() {bUseCustomPrimitiveData=true, parameterName=PivotCord, channelNames {a="", b=Z, g=Y, r=X}, defaultValue=<rgba 0 0 0 1>}
  NamedRerouteDeclaration(None=NaniteReplace!(
    Default=ActorPositionWS() {originType=CameraRelative}
    Nanite=TransformPosition(None=@n45) {transformType=TRANSFORMPOSSOURCE_TranslatedWorld}
   )) {name=Actor Position, nodeColor=<rgba 1 0 0.0545 1>}
  n46 = MaterialFunctionCall(Float3=NamedRerouteUsage()) {materialFunction=&BreakOutFloat3Components}
  n47 = ScalarParameter() {defaultValue=700, group=WindChains, parameterName=Movement Intensity Gradient, sortPriority=3}
  n48 = MaterialFunctionCall(X=@n46.R, Y=@n46.R, Z=Subtract(A=@n46.B, B=@n47) {constB=500}) {materialFunction=&MakeFloat3}
  n49 = Divide(A=Subtract(A=WorldPosition() {worldPositionShaderOffset=WPT_CameraRelativeNoOffsets}.Z, B=ComponentMask(None=@n48.Result) {b=true}), B=@n47) {constB=500}
  NamedRerouteDeclaration(None=Clamp(None=OneMinus(None=@n49)) {clampMode=CMODE_ClampMin}) {name=Gradient, nodeColor=<rgba 0.4766 1 0 1>}
  NamedRerouteDeclaration(None=@n40.MaterialAttributes) {name=MA, nodeColor=<rgba 1 0 0.8906 1>}
  NamedRerouteDeclaration(None=Constant()) {name=Storm-Gust Weight, nodeColor=<rgba 0.5938 0 1 1>}
  NamedRerouteDeclaration(None=@n40.WorldPositionOffset) {name=WPO O, nodeColor=<rgba 0 0.8125 1 1>}
  NamedRerouteDeclaration(None=@n13) {name=Wooble, nodeColor=<rgba 0.8047 1 0 1>}
  n50 = ScalarParameter() {bUseCustomPrimitiveData=true, defaultValue=100, group=WindChains, parameterName=Fixture Length, primitiveDataIndex=4, sortPriority=3}
  NamedRerouteDeclaration(None=@n50) {name=Fixture Lenght, desc="BP ", nodeColor=<rgba 0 0.4375 1 1>}
  n51 = MaterialFunctionCall(Float3=NamedRerouteUsage()) {materialFunction=&BreakOutFloat3Components}
  n52 = MaterialFunctionCall(X=@n51.R, Y=@n51.R, Z=Subtract(A=@n51.B, B=Constant() {desc=Gradient Start Position, r=800}) {constB=500}) {materialFunction=&MakeFloat3}
  n53 = Divide(
   A=Subtract(A=WorldPosition() {worldPositionShaderOffset=WPT_CameraRelativeNoOffsets}.Z, B=ComponentMask(None=@n52.Result) {b=true})
   B=Multiply(A=@n50, B=Constant() {r=0.5}) {constB=0.6}
  ) {constB=500}
  n54 = Constant() {r=0.2}
  NamedRerouteDeclaration(None=Saturate(None=MaterialFunctionCall(In=@n53, Contrast=@n54) {materialFunction=&CheapContrast}.Result)) {name=Gradient 2, nodeColor=<rgba 0.177 0.3906 0 1>}
  NamedRerouteDeclaration(None=Multiply(A=ScalarParameter() {defaultValue=0.5, group=WindChains, parameterName=Swing Intensity, sliderMax=1, sortPriority=2}) {constB=0.5}) {name=Swing Int, nodeColor=<rgba 0 1 0.1797 1>}
  StaticSwitchParameter(
   True=ScalarParameter() {defaultValue=1, group=WindChains, parameterName=STORM, sliderMax=1}
   False=MaterialFunctionCall!(Position, IsTrees, IsGrass, IsMediumPlant, IsMediumPlant_2024Anim, IsMediumPlant2ndLevel, MediumPlant_BasePivotPostion) {materialFunction=/Game/Environments/Asteria/Foliage/Wind/MF_WindActorData}.StormMask
  ) {group=WindChains, parameterName=Debug STORM, sortPriority=0}
 ]}
